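# Niagara effect of green-tinted glowing embers, named for Hera's DivinePlow, using one GPU sprite emitter.
# Randomized loops spawn particles in a cylinder with velocity from an offset point; spawn size shrinks across each loop.
# Upward acceleration, gusting wind and drag move the embers, which stretch with speed as HDR color dims and alpha fades.
# Distance scaling reduces spawn count and enlarges sprites toward the cull range.
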NiagaraSystem /Game/Environments/Hera/FX/DivinePlow/NS_DivinePlow_Embers
dependencies {
 usedDataInterfaces [
  &NiagaraDataInterfaceCurve
  &NiagaraDataInterfaceCollisionQuery
  &NiagaraDataInterfaceLandscape
  &NiagaraDataInterfaceColorCurve
  &NiagaraDataInterfaceArrayFloat
 ]
 usedDynamicInputs []
 usedModules [
  &SystemState
  &self:SystemSpawnScript.SystemScriptSource.SystemScriptGraph.NiagaraNodeAssignment_0.SetVariables_#1
  &Timeline
 ]
}
emitters {Embers {
  simTarget=GPUComputeSim
  emitterUpdateScript [
   &EmitterState(
    Enable Distance Culling=true
    Life Cycle Mode=::NewEnumerator1
    Loop Duration=&RandomRangeFloat(Minimum=2, Maximum=3)
    MaxDistance=15000
    Module.bApplySpawnCountScaleByDistance=true
    Module.bUseMaxDistance=true
    Module.bUseMinDistance=false
    Recalculate Duration Each Loop=true
    Scalability Mode=::NewEnumerator1
    Scale Spawn Count=true
    Spawn Count Scale By Distance=DataInterface() {propertyValues {Curve {keys [<at 0 1>, {arriveTangent=-3.2684, interpMode=RCIM_Cubic, leaveTangent=-3.2684, tangentMode=RCTM_User, time=0.5, value=1}, <at 1 0>]}}}
    UseLoopDelay=true
   )
   &SpawnRate(SpawnRate=50, Use Spawn Probability=true)
  ]
  particleSpawnScript [
   &InitializeParticle(
    Lifetime=&UniformRangedFloat(Minimum=0.25, Maximum=1.5)
    Mass Mode=::NewEnumerator1
    Position=&SimulationPosition()
    Position Mode=::NewEnumerator1
    Sprite Size=&Vector2DFromFloat(Value=&FloatFromCurve(
      FloatCurve=DataInterface() {propertyValues {Curve {keys [<at 0 1.2>,<at 1 0.5>]}}}
      CurveIndex=$Emitter.NormalizedLoopAge
      Scale Curve=&RandomRangeFloat(Minimum=1.5, Maximum=3)
     ))
    Sprite Size Mode=::NewEnumerator4
    UsePositionOffset=false
    Write Material Random=true
   ) {module=InitializeParticle001}
   &AddVelocityFromPoint(Module.Use Falloff Distance=false, Origin Offset=<0 -200 -50>, Velocity Strength=50)
   &ShapeLocation(
    Cylinder Height=50
    Cylinder Radius=40
    MaskSpawnGroup=false
    Override Seed=false
    Shape Primitive=::NewEnumerator1
    Surface Only=false
    UseOwnerScale=true
   )
  ]
  particleUpdateScript [
   &ParticleState()
   &ScaleSpriteSize(Uniform Curve Scale=&FN_LOD_Distance_Remap(Start Distance=0, End Distance=15000, Minimum Output=3, Maximum Output=10, Clamp Output=true))
   &AccelerationForce(Acceleration=&Multiply_Vector(
     A=&VectorFromFloat(Value=1)
     B=&UniformRangedVector(Minimum=<0 0 75>, Maximum=<0 0 150>)
    ))
   &WindForce(
    Frequency Min / Max:Vector2f=<1 5>
    Friction Mode:ENiagaraWindFrictionMode=::NewEnumerator0
    Scale Min / Max:Vector2f=<1 0.25>
    Speed Min / Max:Vector2f=<0 400>
    Surface Alignment Mode:ENiagaraWindCombingMode=::NewEnumerator0
    Turbulence Frequency Mode:ENiagaraWindTurbulenceFrequencyMode=::NewEnumerator1
    Turbulence Scale Mode:ENiagaraWindTurbulenceContributionMode=::NewEnumerator1
    Wind Speed=&RandomRangeVector(Minimum=<0 0 125>, Maximum=<100 1 250>)
   )
   &WindForce(
    Scale Min / Max:Vector2f=<1 3>
    Speed Min / Max:Vector2f=<75 500>
    Turbulence Frequency Mode:ENiagaraWindTurbulenceFrequencyMode=::NewEnumerator1
    Turbulence Scale Mode:ENiagaraWindTurbulenceContributionMode=::NewEnumerator1
   ) {enabled=false, module=WindForce001}
   &AerodynamicDrag(Aerodynamic Drag=2, Lift Contribution=0, Use Linear Drag=true, Use Rotational Drag=true)
   &SolveForcesAndVelocity(Acceleration Limit=100, Clamp Velocity=true, Limit Acceleration=true)
   &ScaleSpriteSizeBySpeed(
    Max Scale Factor=<1.5 3>
    Sample Scale Factor By Curve=true
    Scale Factor Curve=DataInterface() {propertyValues {Curve {keys [
        <at -0.0077 0.0765 RCIM_Cubic>
        {arriveTangent=1.2463, interpMode=RCIM_Cubic, leaveTangent=1.2463, tangentMode=RCTM_User, time=0.9985, value=0.9549}
       ]}}}
   )
   &AlignSpriteToMeshOrientation(SetAlignment=true, SetFacing=true)
   &ScaleColor(
    Scale Alpha=&FloatFromCurve(FloatCurve=DataInterface() {propertyValues {Curve {keys [<at 0 1>,<at 1 0>]}}})
    Scale RGB=&Multiply_VectorByFloat(
     Vector=&MakeVectorFromLinearColorRGB(Color=&ColorFromCurve(ColorCurve=DataInterface() {propertyValues {
         AlphaCurve {keys [<at 0 1 RCIM_Cubic>,<at 0.7107 1>]}
         BlueCurve {keys [<at 0 14.6701 RCIM_Cubic>,<at 0.2499 1.5104>,<at 0.7114 0.1753 RCIM_Cubic>]}
         GreenCurve {keys [<at 0 50 RCIM_Cubic>,<at 0.2499 10>,<at 0.7114 1 RCIM_Cubic>]}
         RedCurve {keys [<at 0 17.6143 RCIM_Cubic>,<at 0.2499 2.2179>,<at 0.7114 0.2441 RCIM_Cubic>]}
        }}))
     Float=5
    )
    ScaleA=true
    ScaleRGB=true
   )
   /Niagara/Modules/Update/Camera/MaintainInCameraParticleScale(Pixel Depth=5) {enabled=false}
  ]
  rendererClasses [&NiagaraSpriteRendererProperties]
  renderers [{rendererClass=&NiagaraSpriteRendererProperties}]
 }}
summary {
 systemName=NS_DivinePlow_Embers
 emitters [{emitterName=Embers, simTarget=GPUComputeSim, rendererClasses [&NiagaraSpriteRendererProperties]}]
 userVariables []
}
